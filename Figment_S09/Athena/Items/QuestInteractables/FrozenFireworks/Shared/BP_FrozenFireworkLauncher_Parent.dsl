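; Parent quest interactable for Frozen Fireworks launchers that lets players with an unfinished quest objective launch a firework.
; Interaction locks the prop and starts fuse effects, with a timeline moving the sparks upward.
; When the fuse finishes, a curve-table hotfix value read at BeginPlay gates the launch.
; Value 0 plays local-player sound, spawns BP_FrozenFirework at random speed, then hides the prop with optional shrink; value 1 skips all this.
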(event EventBeginPlay
  ;[1 1 1] Do parent stuff like hide/bind quest updates
  (|CallParentFunction)
  (bind _outxy (Fortnite|EvaluateCurveTableRow (Variables|Default|GetHotFixVal)))
  (bind _output_get (Variables|Default|SetHotFixOnOff (Math|Float|Truncate _outxy)))
  (Development|PrintString (Utilities|String|ToString(Integer) _output_get) false false))

(event Interaction|EventBlueprintOnInteract
  (params (InteractingPawn :type Fort%20Pawn%20Object%20Reference) (InteractionBeingAttempted :type EInteractionBeingAttempted%20Enum :default "FirstInteraction"))
  ;[1 1 1] Only server stuff off this.
  (Variables|Default|SetPlayerPawn InteractingPawn)
  (Variables|Default|SetbInteractLocked true)
  (CallFunction|StartFuse))

(event Custom|UnlockInteract
  ;[1 1 1] If we ever want to use this, not currently needed
  (Variables|Default|SetbInteractLocked false))

(event Custom|StartFuse
  (bind _fusesparks (Variables|Default|GetFuseSparks))
  ;[1 1 1] Fuse FX and sound
  (Components|Activation|Activate (Variables|Default|GetInitialFuse) false)
  (Utilities|FlowControl|Delay 0.1)
  (Variables|Default|SetStartingSparkLoc (Variables|/Script/Engine.SceneComponent|GetRelativeLocation _fusesparks))
  (Components|Activation|Activate _fusesparks false)
  (Components|Activation|Activate (Variables|Default|GetFuse) false)
  (bind (_direction _fusemovementup) (Timeline|Timeline_1 PlayFromStart)
    (:Update
      (Transformation|SetRelativeLocation _fusesparks (Math|Vector|Lerp(Vector) (Variables|Default|GetStartingSparkLoc) (Math|Vector|MakeVector (.x (Variables|Default|GetStartingSparkLoc)) (.y (Variables|Default|GetStartingSparkLoc)) (+ (.z (Variables|Default|GetStartingSparkLoc)) 33.0)) _fusemovementup) false false))
    (:Finished
      ;[1 1 1] Launch Sound and FX
      (Components|Activation|Deactivate (Variables|Default|GetFuse))
      (Components|Activation|Deactivate (Variables|Default|GetFuseSparks))
      (switch int (Variables|Default|GetHotFixOnOff)
        (:0
          (if (Utilities|Equal(Object) (Pawn|GetController (Variables|Default|GetPlayerPawn)) (Utilities|Array|Get(acopy) (Fortnite|GetLocalFortPlayerControllers true)))
            (Audio|PlaySoundatLocation (Variables|Default|GetFireRocketSound) (Transformation|GetWorldLocation (Variables|Debug|GetStaticMeshComponent)))
            (else
              (Development|PrintString "SkipSound If Hotfixed Off and not local" false false)))
          ;[1 1 1] Spawn Firework
          (Rendering|SetVisibility (Variables|Default|GetFireworkMesh) false false)
          (Projectile|SpawnProjectile :ProjectileClass "/Figment_S09/Athena/Items/QuestInteractables/FrozenFireworks/Shared/BP_FrozenFirework.BP_FrozenFirework_C" :RequestedBy (Variables|Default|GetPlayerPawn) :SpawnLocation (Transformation|GetWorldLocation (Variables|Default|GetMuzzle)) :SpawnRotation (Transformation|GetWorldRotation (Variables|Debug|GetStaticMeshComponent)) :SpawnDirection (Transformation|GetWorldRotation (Variables|Default|GetArrow)) :bUseDefaultPhysics false :InitialSpeed (Math|Random|RandomFloatinRange 4.5e+03 5e+03) :GravityScale 0.4 :ChargePercent 1.0)
          ;[1 1 1] Should Shrink Before Hiding?
          (if (Variables|Default|GetMeshShrinkOnSuccess)
            (bind (_direction_2 _shrinktime) (Timeline|Timeline_0 PlayFromStart)
              (:Update
                (Transformation|SetRelativeScale3D (Variables|Debug|GetStaticMeshComponent) (Math|Vector|Lerp(Vector) "1.000000,1.000000,1.000000" "0.010000,0.010000,0.010000" _shrinktime)))
              (:Finished
                _))
            (else
              (if (Variables|Default|GetLetChildBPHideProp)
                _
                (else
                  _))))
          (CallFunction|HideQuestProp))
        (:1
          _)))))

(event EventPlaySuccessFX)

(fn BlueprintCanInteract
  (params (InteractingPawn :type Fort%20Pawn%20Object%20Reference) (InteractionBeingAttempted :type EInteractionBeingAttempted%20Enum :default "FirstInteraction") (InteractionType :type TInteractionType%20Enum :default "IT_NoInteraction"))
  (bind _returnvalue (Game|GetInstigatorController InteractingPawn))
  (if (not (Variables|Default|GetbInteractLocked))
    (Utilities|IsValid InteractingPawn
      (:Is%20Valid
        (Utilities|IsValid _returnvalue
          (:Is%20Valid
            (bind _asfort_player_controller_athena (Utilities|Casting|CastToFortPlayerControllerAthena _returnvalue))
            (bind (_questisvalid _questcompleted _objectivecompleted) (Default|AthenaCheckQuestandObjectiveCompletion _asfort_player_controller_athena (Variables|QuestStuff|GetQuestItem) (Variables|Default|GetQuestBackendObjectiveName)))
            (return (and (and _questisvalid (not _questcompleted)) (not _objectivecompleted)))))))))

(fn BlueprintGetInteractionString
  (params (InteractingPawn :type Fort%20Pawn%20Object%20Reference) (InteractionBeingAttempted :type EInteractionBeingAttempted%20Enum :default "FirstInteraction"))
  (return "NSLOCTEXT(\"[F784A61D47105718CC8A748C78B35E65]\", \"D71BAB964D6F73121CD23A850C5AC025\", \"Launch firework\")"))

(fn ConstructionScript
  (|CallParentFunction)
  (Variables|Default|SetNewVar_1 (Math|Transform|MakeTransform (Transformation|GetWorldLocation (Variables|Default|GetArrow)) (Transformation|GetWorldRotation (Variables|Default|GetArrow)) (Variables|/Script/Engine.SceneComponent|GetRelativeScale3D (Variables|Default|GetFireworkMesh)))))
